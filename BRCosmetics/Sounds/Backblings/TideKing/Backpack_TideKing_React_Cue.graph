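# Sound cue that plays a reaction sound for the TideKing back bling cosmetic.
# A short fixed delay feeds a modulator that adds slight random pitch and volume variation to an equal-weight random pick of three wave variants.
# Output is spatialized with the close-range pet attenuation, routed through the back bling sound class, and uses the reactive back bling audio property sheet.
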SoundCue /BRCosmetics/Sounds/Backblings/TideKing/Backpack_TideKing_React_Cue
root=Delay_0
nodes [
 {id=Delay_0, children [Modulator_0], properties {delayMax=0.3, delayMin=0.3}}
 {id=Modulator_0, children [Random_0], properties {pitchMax=1.02, pitchMin=0.98, volumeMax=1.02, volumeMin=0.98}}
 {id=Random_0, children [WavePlayer_1,WavePlayer_2,WavePlayer_3], properties {weights [1,1,1]}}
 {id=WavePlayer_1, properties {soundWaveAssetPtr=&/Backpack_TideKing_React_01}}
 {id=WavePlayer_2, properties {soundWaveAssetPtr=&/Backpack_TideKing_React_02}}
 {id=WavePlayer_3, properties {soundWaveAssetPtr=&/Backpack_TideKing_React_03}}
]
properties {
 attenuationSettings=&Fort_Pets_Close_Athena_Att
 audioPropertiesSheet=/BRCosmetics/Sounds/Settings/PropertySheets/Reactive/APS_Cosmetic_Reactive_Backbling
 maxDistance=1800.0001
 soundClassObject=/Game/Sounds/FortSoundClassesAndModes/Fort_Backblings
 volumeMultiplier=0.55
}
